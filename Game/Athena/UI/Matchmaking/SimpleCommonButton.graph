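# SimpleCommonButton is a bare Widget Blueprint button based on CommonButtonLegacy, kept in the Athena matchmaking UI folder.
# Its widget tree holds an empty CanvasPanel root, plus an inherited InputActionWidget slot with no widget assigned.
# The PreConstruct, Construct and Tick events are unconnected stubs and it declares no variables, so the graph adds no behavior beyond the parent class.
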WidgetBlueprint /Game/Athena/UI/Matchmaking/SimpleCommonButton
graphs [EventGraph {nodes [
   Event PreConstruct()
   Event Construct()
   Event Tick()
  ]}]
variables []
widgets {
 info {inheritedWidgetCount=1, parentClass=&CommonButtonLegacy, rootWidgetClass=&CanvasPanel, widgetCount=2}
 widgets [
  CanvasPanel_0 {widgetClassPath=&CanvasPanel, widget=CanvasPanel_0}
  InputActionWidget {widgetClassPath=&CommonActionWidget, bInherited=true, widget=null}
 ]
}
