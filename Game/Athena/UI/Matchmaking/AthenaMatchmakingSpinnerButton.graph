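# Fortnite spinner row widget for matchmaking and Creative playlist options, built on FortMatchmakingKnobsSpinnerButton.
# Arrow clicks slide the value text out, shift it, then reverse the opposite animation, while input type drives sizing, fonts and arrow visibility.
# OneTimeChange-tagged options are enabled only while the volume's link ID is empty, preloading a privacy warning shown modally on rotation.
# Device button options show a button that broadcasts clicks through the player controller.
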WidgetBlueprint /Game/Athena/UI/Matchmaking/AthenaMatchmakingSpinnerButton
animations [
 Left {
  binding_count=1
  display_label=""
  length_seconds=0.1
  object=&self:Left
  sequence {bindings [MyText {
     id=#1
     possessed_class=CommonTextBlock
     tracks [
      {
       type=MovieScene2DTransformTrack
       sections [{
         type=MovieScene2DTransformSection
         end_frame=2
         start_frame=0
         channels [
          Translation.X {type=MovieSceneScriptingFloatChannel, keys [[0,0],[2,-100]]}
          Translation.Y {type=MovieSceneScriptingFloatChannel, const [0,1]}
          Angle {type=MovieSceneScriptingFloatChannel, const [0,1]}
          Scale.X {type=MovieSceneScriptingFloatChannel, const [1,1]}
          Scale.Y {type=MovieSceneScriptingFloatChannel, const [1,1]}
          Shear.X {type=MovieSceneScriptingFloatChannel, const [0,1]}
          Shear.Y {type=MovieSceneScriptingFloatChannel, const [0,1]}
         ]
        }]
      }
      {
       type=MovieSceneFloatTrack
       display_name=Render Opacity
       sections [{type=MovieSceneFloatSection, end_frame=2, start_frame=0, channels [None {type=MovieSceneScriptingFloatChannel, keys [[0,1],[2,0]]}]}]
      }
     ]
    }]}
 }
 Right {
  binding_count=1
  display_label=""
  length_seconds=0.1
  object=&self:Right
  sequence {bindings [MyText {
     id=#1
     possessed_class=CommonTextBlock
     tracks [
      {
       type=MovieScene2DTransformTrack
       sections [{
         type=MovieScene2DTransformSection
         end_frame=2
         start_frame=0
         channels [
          Translation.X {type=MovieSceneScriptingFloatChannel, keys [[0,0],[2,100]]}
          Translation.Y {type=MovieSceneScriptingFloatChannel, const [0,1]}
          Angle {type=MovieSceneScriptingFloatChannel, const [0,1]}
          Scale.X {type=MovieSceneScriptingFloatChannel, const [1,1]}
          Scale.Y {type=MovieSceneScriptingFloatChannel, const [1,1]}
          Shear.X {type=MovieSceneScriptingFloatChannel, const [0,1]}
          Shear.Y {type=MovieSceneScriptingFloatChannel, const [0,1]}
         ]
        }]
      }
      {
       type=MovieSceneFloatTrack
       display_name=Render Opacity
       sections [{type=MovieSceneFloatSection, end_frame=2, start_frame=0, channels [None {type=MovieSceneScriptingFloatChannel, keys [[0,1],[2,0]]}]}]
      }
     ]
    }]}
 }
 In {
  binding_count=1
  length_seconds=0.35
  object=&self:In
  sequence {bindings [Border_Fill {
     id=#2
     possessed_class=Border
     tracks [
      {
       type=MovieScene2DTransformTrack
       sections [{
         type=MovieScene2DTransformSection
         end_frame=7
         start_frame=0
         channels [
          Translation.X {type=MovieSceneScriptingFloatChannel, keys [[0,300],[5,-25],[7,0]]}
          Translation.Y {type=MovieSceneScriptingFloatChannel}
          Angle {type=MovieSceneScriptingFloatChannel}
          Scale.X {type=MovieSceneScriptingFloatChannel}
          Scale.Y {type=MovieSceneScriptingFloatChannel}
          Shear.X {type=MovieSceneScriptingFloatChannel}
          Shear.Y {type=MovieSceneScriptingFloatChannel}
         ]
        }]
      }
      {
       type=MovieSceneFloatTrack
       display_name=Render Opacity
       sections [{type=MovieSceneFloatSection, end_frame=7, start_frame=0, channels [None {type=MovieSceneScriptingFloatChannel, keys [[0,0],[7,1]]}]}]
      }
     ]
    }]}
 }
]
graphs [
 BP_InitButtonOption {nodes [
   n1 = BP_InitButtonOption!() -> (then, ButtonComponent:Fort Creative Device Button Component Object Reference) {is=title K2Node_FunctionEntry}
   > Set Button Component(Button Component:Fort Creative Device Button Component Object Reference=@n1) -> (Output_Get:Fort Creative Device Button Component Object Reference)
   > SetHorizontalAlignment(
    self:Size Box Slot Object Reference=SlotAsSizeBoxSlot!(Widget:Widget Object Reference=Get Overlay_1() -> (Overlay_1:Overlay Object Reference)) -> (ReturnValue:Size Box Slot Object Reference)
    InHorizontalAlignment=HAlign_Center
   )
   > SetVisibility(self=Get SizeBox_Button() -> (SizeBox_Button:Size Box Object Reference), InVisibility=Visible)
   > Bind Event to On Button Clicked(self:Common Button Legacy Object Reference=Get Button_Widget() -> (Button_Widget:Icon Text Button Object Reference), Delegate=Create Event()) {is=title K2Node_AddDelegate}
   > Set Text!(
    execute
    self:Icon Text Button Object Reference=Get Button_Widget()
    ButtonText:Text=GetButtonLabel!(self:Fort Creative Device Button Component Object Reference=Get ButtonComponent() -> (ButtonComponent:Fort Creative Device Button Component Object Reference)) -> (ReturnValue:Text)
   ) -> (then) {is=title K2Node_CallFunction}
  ]}
 DoCustomNavigation_0 {nodes [
   DoCustomNavigation_0!() -> (then, Navigation:EUINavigation Enum) {is=title K2Node_FunctionEntry}
   > ShiftTextLeft(self:Common Rotator Object Reference)
   > Return Node!(execute, ReturnValue:Widget Object Reference=Self-Reference())
  ]}
 EventGraph {nodes [
   n1 = "On Button Clicked (LeftButton)"!() -> (OutputDelegate, then, Button:Common Button Legacy Object Reference) {is=title K2Node_ComponentBoundEvent}
   n2 = AfterleftisDone() {is=title K2Node_CustomEvent}
   n3 = Get Left() -> (Left:Widget Animation Object Reference)
   BindToAnimationFinished!(execute=@n1.then, self:Widget Animation Object Reference=@n3, Widget:User Widget Object Reference=Self-Reference(), Delegate=@n2) -> (then)
   > n4 = PlayAnimation(InAnimation=@n3)
   Branch(execute=[@n4.then, @n8.then], Condition=Get isTouch() -> (isTouch:bool))
   > Highlight!(execute, self, Highlight:bool=true) -> (then)
   n5 = "On Button Clicked (RightButton)"!() -> (OutputDelegate, then, Button:Common Button Legacy Object Reference) {is=title K2Node_ComponentBoundEvent}
   n6 = AfterRightIsDone() {is=title K2Node_CustomEvent}
   n7 = Get Right() -> (Right:Widget Animation Object Reference)
   BindToAnimationFinished(execute=@n5.then, self=@n7, Widget=Self-Reference(), Delegate=@n6)
   > n8 = PlayAnimation(InAnimation=@n7)
   n9 = OnHandleOptionGenerated_Event!() -> (OutputDelegate, then, InPlaylistUserOption:Playlist User Option Base Object Reference) {is=title K2Node_CustomEvent}
   > n10 = Update Enabled State!(
    execute
    self
    InPlaylistUserOptionTagContainer:Gameplay Tag Container Structure=Get Tags!(self:Playlist User Option Base Object Reference=@n9.InPlaylistUserOption) -> (Tags:Gameplay Tag Container Structure)
   ) -> (then, ShouldShowPrompt:bool)
   > n11 = Branch(Condition=NOT Boolean(A=@n10))
   n12 = Get ValueChangeWidget() -> (ValueChangeWidget:User Widget Object Reference)
   > Is Valid(InputObject=@n12)
   > RemoveFromParent(self=@n12)
   > Set ValueChangeWidget(ValueChangeWidget:User Widget Object Reference) -> (Output_Get:User Widget Object Reference)
   n13 = Is Valid(exec=@n11.else, InputObject=Get ValueChangeWidget())
   n14 = CreateWidgetAsync!(
    execute:Exec=@n13.Is Not Valid
    UserWidgetSoftClass:User Widget Soft Class Reference=/Game/UI/Creative/Creative_PrivacyWarningWidget.Creative_PrivacyWarningWidget_C
    OwningPlayer:Player Controller Object Reference=GetOwningPlayer()
    bSuspendInputUntilComplete:bool=true
   ) -> (then:Exec, OnComplete:Exec, OnFailure:Exec, UserWidget:User Widget Object Reference) {is=title K2Node_AsyncAction}
   Set ValueChangeWidget(execute=@n14.OnComplete, ValueChangeWidget=@n14)
   n15 = InputUpdated!() -> (OutputDelegate, then, bNewInputType:ECommonInputType Enum) {is=title K2Node_CustomEvent}
   > n16 = UpdateSyle!(execute, self, InputType:ECommonInputType Enum=@n15.bNewInputType) -> (then)
   Set Button Sizes(execute=[@n25.then, @n16.then]) {is=title K2Node_CallFunction}
   > n17 = Bind Event to On Handle Option Widget Generated(self:Fort Matchmaking Knobs Spinner Button Object Reference, Delegate=@n9) {is=title K2Node_AddDelegate}
   n18 = OnKnobDataInitialized_Event!() -> (OutputDelegate, then, InPlaylistUserOption:Playlist User Option Base Object Reference) {is=title K2Node_CustomEvent}
   Assign On Knob Data Initialized!(execute:Exec=@n17.then, self:Fort Matchmaking Knobs Spinner Button Object Reference, Delegate:"Delegate (by ref)"=@n18) -> (then:Exec) {is=title K2Node_AssignDelegate}
   n19 = AnimateIn!() -> (OutputDelegate, then, Index:int) {is=title K2Node_CustomEvent}
   > SetRenderOpacity(InOpacity=0)
   > Delay(Duration=int * float!(A:int=@n19.Index, B:double=0.03) -> (ReturnValue:double))
   > PlayAnimation(InAnimation=Get In() -> (In:Widget Animation Object Reference), PlaybackSpeed=2)
   > SetRenderOpacity(InOpacity=1)
   Event On Clicked()
   > n20 = Branch(Condition=Get isTouch())
   > Highlight(Highlight=true)
   ShiftTextRight(execute=@n20.else, self:Common Rotator Object Reference)
   Event On Hovered()
   > n21 = Branch(Condition=Get isTouch())
   > SetArrowVisibility()
   Highlight(execute=@n21.else, Highlight=true)
   n22 = Event OnLoadingStateChanged!() -> (OutputDelegate, then, bCurrentlyLoading:bool) {is=title K2Node_Event}
   > SetVisibility(
    self=Get SaveSpinner() -> (SaveSpinner:Image Object Reference)
    InVisibility=Select(Option 0=Collapsed, Option 1=Visible, Index=@n22.bCurrentlyLoading)
   )
   Event OnFocusLost!() -> (OutputDelegate, then, InFocusEvent:Focus Event Structure) {is=title K2Node_Event}
   > Branch(Condition=Get isTouch())
   > Highlight(Highlight=false)
   Event On Unhovered()
   > n23 = Branch(Condition=Get isTouch())
   Highlight(execute=@n23.else, Highlight=false)
   Event Construct()
   n24 = Get CommonInputSubsystem!() -> (ReturnValue:Common Input Object Reference) {is=title K2Node_GetBlueprintContextEx}
   > UpdateSyle(InputType=GetCurrentInputType(self=@n24))
   > n25 = Assign On Input Method Changed!(execute:Exec, self:Common Input Object Reference=@n24, Delegate:"Delegate (by ref)"=@n15) -> (then:Exec) {is=title K2Node_AssignDelegate}
   n26 = Event OnSpinnerRotated!() -> (OutputDelegate, then, Value:int=0, bUserInitiated:bool) {is=title K2Node_Event}
   > Call On Spinner Value Changed!(execute:Exec, self:Athena Matchmaking Spinner Button Object Reference, Index:int=@n26.Value) -> (then:Exec) {is=title K2Node_CallDelegate}
   n27 = Get ValueChangeWidget()
   > Is Valid(InputObject=@n27)
   > TryPushWidgetAsModalPanelForPlayer!(
    execute
    LocalPlayerController:Fort Player Controller Object Reference=Cast To FortPlayerController!(Object=GetOwningPlayer()) -> (AsFort Player Controller, bSuccess:bool=~)
    PushWidget:User Widget Object Reference=@n27
   ) -> (then)
   n28 = Event SetSpinnerOptionsVisiblity!() -> (OutputDelegate, then, bVisible:bool) {is=title K2Node_Event}
   > n29 = Branch(Condition=@n28.bVisible)
   > n30 = Set VisibilitySetting(VisibilitySetting:ESlateVisibility Enum=Visible) -> (Output_Get:ESlateVisibility Enum=Visible)
   SetVisibility(
    execute=[@n31.then, @n30.then]
    self=Get OptionSelectionArea() -> (OptionSelectionArea:Horizontal Box Object Reference)
    InVisibility=Get VisibilitySetting() -> (VisibilitySetting:ESlateVisibility Enum=Visible)
   )
   n31 = Set VisibilitySetting(execute=@n29.else, VisibilitySetting=Collapsed) -> (Output_Get=Visible)
   Event Destruct()
   > Unbind Event from On Handle Option Widget Generated(self:Fort Matchmaking Knobs Spinner Button Object Reference, Delegate=Create Event()) {is=title K2Node_RemoveDelegate}
   n32 = Get ValueChangeWidget()
   > Is Valid(InputObject=@n32)
   > RemoveFromParent(self=@n32)
   > Set ValueChangeWidget()
   n33 = Get Left()
   PlayAnimation(execute=@n37.then, InAnimation=@n33, PlayMode=Reverse)
   n34 = Get Right()
   PlayAnimation(execute=@n36.then, InAnimation=@n34, PlayMode=Reverse)
   SetVisibility(execute=@n41.then, self=Get SizeBox_Button(), InVisibility=Collapsed)
   n35 = Self-Reference()
   n36 = UnbindAllFromAnimationFinished!(execute=@n39.then, self:Widget Animation Object Reference=[@n33, @n34], Widget:User Widget Object Reference=@n35) -> (then)
   n37 = UnbindAllFromAnimationFinished(execute=@n38.then, self=[@n33, @n34], Widget=@n35)
   n38 = ShiftTextRight(execute=@n6.then)
   n39 = ShiftTextLeft(execute=@n2.then)
   n40 = SlotAsSizeBoxSlot(Widget=Get Overlay_1())
   n41 = SetHorizontalAlignment(execute=@n18.then, self=@n40, InHorizontalAlignment=HAlign_Right)
  ]}
 GetDeviceButtonComponent {nodes [
   Inputs()
   > n1 = Cast To Actor!(execute, Object=GetInitializingObject!(self:Fort Matchmaking Knobs Spinner Button Object Reference) -> (ReturnValue:Object Reference)) -> (then, CastFailed, AsActor:Actor Object Reference) {is=title K2Node_DynamicCast}
   n2 = Local Wildcard!() -> (Variable:Fort Creative Device Button Component Object Reference) {is=title K2Node_TemporaryVariable}
   > Assign!(
    execute:Exec
    Variable:Fort Creative Device Button Component Object Reference=@n2
    Value:Fort Creative Device Button Component Object Reference=GetComponentByClass!(self:Actor Object Reference=@n1, ComponentClass:Actor Component Class Reference=/Script/FortniteGame.FortCreativeDeviceButtonComponent) -> (ReturnValue:Fort Creative Device Button Component Object Reference)
   ) -> (then:Exec) {is=title K2Node_AssignmentStatement}
   > n3 = Is Valid(InputObject=@n2)
   > Outputs!(Is Valid:Exec, Is Not Valid:Exec=@n3.Is Not Valid, ReturnValue:Fort Creative Device Button Component Object Reference=@n2)
  ]}
 Highlight {nodes [
   n1 = Highlight!() -> (then, Highlight:bool=~) {is=title K2Node_FunctionEntry}
   > n2 = Branch(Condition=@n1)
   n3 = Get Image_Selection_Hover() -> (Image_Selection_Hover:Image Object Reference)
   > SetVisibility(self=@n3, InVisibility=SelfHitTestInvisible)
   n4 = Get TextBlock_OptionDisplayName() -> (TextBlock_OptionDisplayName:Common Text Object Reference)
   n5 = Get MyText() -> (MyText:Common Text Object Reference)
   > n6 = SetColorAndOpacity(self=[@n4, @n5], InColorAndOpacity=Make Slate Color(SpecifiedColor=(R=0.0015,G=0.0513,B=0.3712,A=1)))
   SetArrowVisibility(execute=[@n6.then, @n7.then])
   SetVisibility(execute=@n2.else, self=@n3, InVisibility=Collapsed)
   > n7 = SetColorAndOpacity(self=[@n4, @n5], InColorAndOpacity=Make Slate Color(SpecifiedColor=(R=0.3176,G=0.7176,B=0.9843,A=1)))
  ]}
 OnDeviceButtonClicked {nodes [
   OnDeviceButtonClicked!() -> (then, Button:Common Button Legacy Object Reference) {is=title K2Node_FunctionEntry}
   > n1 = Cast To FortPlayerControllerAthena(Object=GetOwningPlayer())
   > BroadcastDeviceButtonComponentClickedDelegate!(
    execute
    self:Fort Player Controller Athena Object Reference=@n1
    DeviceButtonComponent:Fort Creative Device Button Component Object Reference=Get Button Component() -> (Button Component:Fort Creative Device Button Component Object Reference)
   ) -> (then)
  ]}
 OnSpinnerValueChanged {nodes [OnSpinnerValueChanged!() -> (then, Index:int) {is=title K2Node_FunctionEntry}]}
 Set Button Sizes {nodes [
   Set Button Sizes() {is=title K2Node_FunctionEntry}
   n1 = Get isTouch()
   n2 = Get LeftButtonSize() -> (LeftButtonSize:Size Box Object Reference)
   n3 = Get RightButtonSize() -> (RightButtonSize:Size Box Object Reference)
   > SetHeightOverride(self=[@n2, @n3], InHeightOverride=Select(Option 0:double=30, Option 1:double=86, Index=@n1) -> (ReturnValue:double))
   > SetWidthOverride(self=[@n2, @n3], InWidthOverride=Select(Option 0=24, Option 1=64, Index=@n1))
   > SetSize!(
    execute
    self:Spacer Object Reference=Get Spacer_Height() -> (Spacer_Height:Spacer Object Reference)
    InSize:Vector 2D Structure=Select(Option 0:Vector 2D Structure, Option 1:Vector 2D Structure=(X=0,Y=100), Index=Get isTouch()) -> (ReturnValue:Vector 2D Structure)
   ) -> (then)
  ]}
 SetArrowVisibility {nodes [
   SetArrowVisibility() {is=title K2Node_FunctionEntry}
   n1 = IsHovered()
   n2 = Get isTouch()
   > SetVisibility(
    self=[
     Get LeftButton() -> (LeftButton:Simple Common Button Object Reference)
     Get RightButton() -> (RightButton:Simple Common Button Object Reference)
    ]
    InVisibility=Select(Option 0:ESlateVisibility Enum=Hidden, Option 1:ESlateVisibility Enum=SelfHitTestInvisible, Index=OR Boolean(A=@n2, B=@n1)) -> (ReturnValue:ESlateVisibility Enum)
   )
  ]}
 Update Enabled State {nodes [
   n1 = Update Enabled State!() -> (then, InPlaylistUserOptionTagContainer:Gameplay Tag Container Structure) {is=title K2Node_FunctionEntry}
   n2 = MakeLiteralGameplayTagContainer!(Value:Gameplay Tag Container Structure=(GameplayTags=((TagName="Creative.Property.OneTimeChange")))) -> (ReturnValue:Gameplay Tag Container Structure)
   > n3 = Branch(Condition=NOT Boolean(A=HasAnyTags!(
      TagContainer:"Gameplay Tag Container Structure (by ref)"=@n1
      OtherContainer:"Gameplay Tag Container Structure (by ref)"=@n2
      bExactMatch:bool=false
     ) -> (ReturnValue:bool)))
   SetIsEnabled(execute=[@n8.Is Not Valid, @n6.Is Not Valid, @n4.CastFailed, @n3.then], bInIsEnabled=true)
   > Return Node!(execute, ShouldShowPrompt:bool=false)
   n4 = Cast To FortPlayerControllerAthena(execute=@n3.else, Object=GetOwningPlayer())
   n5 = GetCurrentVolume!(self:Fort Player Controller Object Reference=@n4) -> (ReturnValue:Fort Volume Object Reference)
   > n6 = Is Valid(InputObject=@n5)
   n7 = GetComponentByClass(self=@n5, ComponentClass=/Script/FortniteGame.FortLevelSaveComponent) -> (ReturnValue:Fort Level Save Component Object Reference)
   > n8 = Is Valid(InputObject=@n7)
   n9 = IsEmpty(InString=Break Online Link Id!(OnlineLinkId:"Online Link Id Structure (by ref)"=Get Loaded Link Id!(self:Fort Level Save Component Object Reference=@n7) -> (LinkId:Online Link Id Structure) {is=title K2Node_CallFunction}) -> (Mnemonic:String, Version:int) {is=title K2Node_BreakStruct})
   > SetIsEnabled(bInIsEnabled=@n9)
   > Return Node(ShouldShowPrompt=@n9)
  ]}
 UpdateSyle {nodes [
   n1 = UpdateSyle!() -> (then, InputType:ECommonInputType Enum) {is=title K2Node_FunctionEntry}
   > Set isTouch(isTouch:bool="Equal (Enum)"(A=@n1, B=Touch)) -> (Output_Get:bool)
   n2 = Get InputType() -> (InputType:ECommonInputType Enum)
   n3 = Get LeftButtonSize()
   n4 = Get RightButtonSize()
   > SetHeightOverride(
    self=[@n3, @n4]
    InHeightOverride=Select!(MouseAndKeyboard:double=30, Gamepad:double=30, Touch:double=56, Count:double=0, Index:ECommonInputType Enum=@n2) -> (ReturnValue:double)
   )
   > SetWidthOverride(self=[@n3, @n4], InWidthOverride=Select(MouseAndKeyboard=24, Gamepad=24, Touch=56, Count=0, Index=@n2))
   n5 = Select(MouseAndKeyboard:int=22, Gamepad:int=22, Touch:int=30, Count:int=0, Index=Get InputType()) -> (ReturnValue:int)
   > SetFont(
    self=[Get TextBlock_OptionDisplayName(), Get MyText()]
    InFontInfo=MakeSlateFontInfo!(
     FontObject:Object Reference=&BurbankSmall
     FontMaterial:Object Reference
     OutlineSettings:Font Outline Settings Structure=(OutlineSize=0,bMiteredCorners=False,bSeparateFillAlpha=False,bApplyOutlineToDropShadows=False,OutlineMaterial=None,OutlineColor=(R=0,G=0,B=0,A=1))
     TypefaceFontName:Name=Bold
     Size:float="To Float (Integer)"(InInt=@n5)
     LetterSpacing:int=0
     SkewAmount:float=0
     bForceMonospaced:bool=false
     bMaterialIsStencil:bool=false
     MonospacedWidth:float=1
    ) -> (ReturnValue:Slate Font Info Structure)
   )
   > SetArrowVisibility() {is=title K2Node_CallFunction}
   > SetDesiredSizeOverride!(
    execute
    self:Image Object Reference=Get Image_Selection_Unselected() -> (Image_Selection_Unselected:Image Object Reference)
    DesiredSize:Vector 2D Structure=Select!(Option 0:Vector 2D Structure=(X=32,Y=56), Option 1:Vector 2D Structure=(X=32,Y=82), Index:bool=Get isTouch()) -> (ReturnValue:Vector 2D Structure)
   ) -> (then)
   > Branch(Condition=Get isTouch())
   > n6 = Branch(Condition=OR Boolean(A=IsHovered(), B=GetSelected!(self:Common Button Base Object Reference) -> (ReturnValue:bool)))
   SetVisibility(execute=@n6.else, self=Get Image_Selection_Hover(), InVisibility=Collapsed)
  ]}
]
variables [
 Button Component {type=FortCreativeDeviceButtonComponent, default=null}
 DefaultStyleGamepad {type=Class, default=/FortUILibrary/Foundation/Buttons/ButtonStyles/ButtonStyle-TextOnlyBase_Empty.ButtonStyle-TextOnlyBase_Empty_C}
 DefaultStyleMouse {type=Class, default=&ButtonStyle-TextOnlyBase_Empty_C}
 isTouch {type=boolean}
 OnSpinnerValueChanged {}
 TextColor_Focused {type=SlateColor, default {colorUseRule=UseColor_Specified, specifiedColor=<rgba 0 0.0887 0.9823 1>}}
 TextColor_Normal {type=SlateColor, default {colorUseRule=UseColor_Specified, specifiedColor=<rgba 0.3095 0.7157 0.9911 1>}}
 ValueChangeWidget {type=UserWidget, default=null}
 VisibilitySetting {type=ESlateVisibility, default=Visible}
]
widgets {
 info {
  inheritedWidgetCount=5
  parentClass=/Script/FortniteUI.FortMatchmakingKnobsSpinnerButton
  rootWidgetClass=&Overlay
  widgetCount=27
 }
 widgets [
  Overlay_0 {widgetClassPath=&Overlay, widget=Overlay_0}
  Spacer_Height {
   widgetClassPath=&Spacer
   bIsVariable=true
   parent=Overlay_0
   widget=Spacer_Height
   slot {class=OverlaySlot, properties {verticalAlignment=VAlign_Center}}
  }
  Image_Selection_Unselected {
   widgetClassPath=&Image
   bIsVariable=true
   parent=Overlay_0
   widget=Image_Selection_Unselected
   properties {
    visibility=HitTestInvisible
    brush {
     bIsDynamicallyLoaded=false
     drawAs=Image
     imageType=NoImage
     mirroring=NoMirror
     resourceName=None
     resourceObject=null
     tiling=NoTile
     imageSize=<32 82>
     margin=<ltrb 0 0 0 0>
     outlineSettings {
      bUseBrushTransparency=false
      roundingType=HalfHeightRadius
      width=0
      color {colorUseRule=UseColor_Specified, specifiedColor=<rgba 0 0 0 0>}
      cornerRadii=<0 0 0 0>
     }
     tintColor {colorUseRule=UseColor_Specified, specifiedColor=<rgba 1 1 1 1>}
     uVRegion {bIsValid=false, max=<0 0>, min=<0 0>}
    }
    colorAndOpacity=<rgba 0.0039 0.2314 0.7686 0.0471>
   }
   slot {class=OverlaySlot, properties {horizontalAlignment=HAlign_Fill, verticalAlignment=VAlign_Center, padding=<ltrb 16 2 16 2>}}
  }
  Image_Selection_Hover {
   widgetClassPath=&Image
   bIsVariable=true
   parent=Overlay_0
   widget=Image_Selection_Hover
   properties {
    visibility=Collapsed
    brush {
     bIsDynamicallyLoaded=false
     drawAs=Image
     imageType=NoImage
     mirroring=NoMirror
     resourceName=None
     resourceObject=/Game/Athena/UI/InGame/Creative/Art/M_SimpleSkewBox_CreativeQuickMenuSelection
     tiling=NoTile
     imageSize=<32 52>
     margin=<ltrb 0 0 0 0>
     outlineSettings {
      bUseBrushTransparency=false
      roundingType=HalfHeightRadius
      width=0
      color {colorUseRule=UseColor_Specified, specifiedColor=<rgba 0 0 0 0>}
      cornerRadii=<0 0 0 0>
     }
     tintColor {colorUseRule=UseColor_Specified, specifiedColor=<rgba 1 1 1 1>}
     uVRegion {bIsValid=false, max=<0 0>, min=<0 0>}
    }
    renderTransform {angle=0, scale=<1.318 1.3>, shear=<0 0>, translation=<0 0>}
   }
   slot {class=OverlaySlot, properties {horizontalAlignment=HAlign_Fill, verticalAlignment=VAlign_Fill}}
  }
  Border_Fill {
   widgetClassPath=&Border
   bIsVariable=true
   parent=Overlay_0
   widget=Border_Fill
   properties {
    navigation=&self:WidgetTree.Border_Fill.WidgetNavigation_0
    background {
     bIsDynamicallyLoaded=false
     drawAs=NoDrawType
     imageType=NoImage
     mirroring=NoMirror
     resourceName=None
     resourceObject=null
     tiling=NoTile
     imageSize=<32 32>
     margin=<ltrb 0 0 0 0>
     outlineSettings {
      bUseBrushTransparency=false
      roundingType=HalfHeightRadius
      width=0
      color {colorUseRule=UseColor_Specified, specifiedColor=<rgba 0 0 0 0>}
      cornerRadii=<0 0 0 0>
     }
     tintColor {colorUseRule=UseColor_Specified, specifiedColor=<rgba 1 1 1 1>}
     uVRegion {bIsValid=false, max=<0 0>, min=<0 0>}
    }
    brushColor=<rgba 1 1 1 0>
    padding=<ltrb 0 0 0 0>
   }
   slot {class=OverlaySlot, properties {horizontalAlignment=HAlign_Fill, verticalAlignment=VAlign_Fill}}
  }
  VerticalBox_0 {
   widgetClassPath=&VerticalBox
   parent=Border_Fill
   widget=VerticalBox_0
   slot {class=BorderSlot, properties {padding=<ltrb 0 0 0 0>}}
  }
  SizeBox_1 {
   widgetClassPath=&SizeBox
   parent=VerticalBox_0
   widget=SizeBox_1
   properties {
    bOverride_MinDesiredHeight=true
    bOverride_MinDesiredWidth=true
    heightOverride=43.6348
    minDesiredHeight=56
    minDesiredWidth=1000
    widthOverride=1000
   }
   slot {class=VerticalBoxSlot, properties {padding=<ltrb 0 0 10 0>, size {sizeRule=Fill, value=1}}}
  }
  CommonBorder_Container {
   widgetClassPath=&CommonBorder
   parent=SizeBox_1
   widget=CommonBorder_Container
   properties {padding=<ltrb 0 0 0 0>}
   slot {class=SizeBoxSlot}
  }
  HorizontalBox_0 {
   widgetClassPath=&HorizontalBox
   parent=CommonBorder_Container
   widget=HorizontalBox_0
   slot {class=BorderSlot, properties {padding=<ltrb 0 0 0 0>}}
  }
  SaveSpinner {
   widgetClassPath=&Image
   bIsVariable=true
   parent=HorizontalBox_0
   widget=SaveSpinner
   properties {
    visibility=Collapsed
    brush {
     bIsDynamicallyLoaded=false
     drawAs=Image
     imageType=NoImage
     mirroring=NoMirror
     resourceName=None
     resourceObject=/FortUILibrary/UIKit/Throbber/MI_UIKit_Color_Throbber
     tiling=NoTile
     imageSize=<35 35>
     margin=<ltrb 0 0 0 0>
     outlineSettings {
      bUseBrushTransparency=false
      roundingType=HalfHeightRadius
      width=0
      color {colorUseRule=UseColor_Specified, specifiedColor=<rgba 0 0 0 0>}
      cornerRadii=<0 0 0 0>
     }
     tintColor {colorUseRule=UseColor_Specified, specifiedColor=<rgba 1 1 1 1>}
     uVRegion {bIsValid=false, max=<0 0>, min=<0 0>}
    }
   }
   slot {class=HorizontalBoxSlot, properties {horizontalAlignment=HAlign_Right, verticalAlignment=VAlign_Center, padding=<ltrb 6 0 0 0>}}
  }
  SizeBox_0 {
   widgetClassPath=&SizeBox
   parent=HorizontalBox_0
   widget=SizeBox_0
   properties {maxDesiredWidth=669.8986, widthOverride=550}
   slot {class=HorizontalBoxSlot, properties {horizontalAlignment=HAlign_Left, padding=<ltrb 24 0 0 0>, size {sizeRule=Fill, value=1}}}
  }
  TextBlock_OptionDisplayName {
   widgetClassPath=&CommonTextBlock
   bInherited=true
   bIsVariable=true
   parent=SizeBox_0
   widget=TextBlock_OptionDisplayName
   properties {
    scrollStyle=&TextScrollStyle-Base_C
    text="[UNINITIALIZED] Á [UNINITIALIZED] Á [UNINITIALIZED]  f fg fg fg fg fg fg f fg fg fg f"
    colorAndOpacity {colorUseRule=UseColor_Specified, specifiedColor=<rgba 0.3185 0.7157 0.9823 1>}
    font {
     bForceMonospaced=false
     bMaterialIsStencil=false
     fontMaterial=null
     fontObject=&BurbankSmall
     letterSpacing=0
     monospacedWidth=1
     size=22
     skewAmount=0
     typefaceFontName=Bold
     outlineSettings {
      bApplyOutlineToDropShadows=false
      bMiteredCorners=false
      bSeparateFillAlpha=false
      outlineMaterial=null
      outlineSize=0
      outlineColor=<rgba 0.0052 0.0052 0.0056 0.2>
     }
    }
   }
   slot {class=SizeBoxSlot, properties {verticalAlignment=VAlign_Center}}
  }
  SizeBox_3 {
   widgetClassPath=&SizeBox
   parent=HorizontalBox_0
   widget=SizeBox_3
   properties {bOverride_WidthOverride=true, widthOverride=420}
   slot {class=HorizontalBoxSlot, properties {horizontalAlignment=HAlign_Right, padding=<ltrb 0 0 10 0>}}
  }
  Overlay_1 {
   widgetClassPath=&Overlay
   bIsVariable=true
   parent=SizeBox_3
   widget=Overlay_1
   slot {class=SizeBoxSlot, properties {horizontalAlignment=HAlign_Right}}
  }
  OptionSelectionArea {
   widgetClassPath=&HorizontalBox
   bIsVariable=true
   parent=Overlay_1
   widget=OptionSelectionArea
   slot {class=OverlaySlot, properties {horizontalAlignment=HAlign_Fill, verticalAlignment=VAlign_Fill}}
  }
  LeftButtonSize {
   widgetClassPath=&SizeBox
   bIsVariable=true
   parent=OptionSelectionArea
   widget=LeftButtonSize
   properties {bOverride_HeightOverride=true, bOverride_WidthOverride=true, heightOverride=30, widthOverride=24}
   slot {class=HorizontalBoxSlot, properties {horizontalAlignment=HAlign_Center, verticalAlignment=VAlign_Center}}
  }
  LeftButton {
   widgetClassPath=&/SimpleCommonButton
   bIsVariable=true
   parent=LeftButtonSize
   widget=LeftButton
   properties {
    onButtonBaseClicked=(&self:WidgetTree.LeftButton.HandleOnButtonClicked)
    onButtonBaseDoubleClicked=(&self:WidgetTree.LeftButton.HandleOnButtonDoubleClicked)
    onButtonBaseHovered=(&self:WidgetTree.LeftButton.HandleOnButtonHovered)
    onButtonBaseUnhovered=(&self:WidgetTree.LeftButton.HandleOnButtonUnhovered)
    onSelectedChangedBase=(&self:WidgetTree.LeftButton.HandleOnSelectedChanged)
    style=/FortUILibrary/Foundation/Buttons/ButtonStyles/ButtonStyle_SimpleCycleArrow_Left.ButtonStyle_SimpleCycleArrow_Left_C
    visibility=Hidden
   }
   slot {class=SizeBoxSlot, properties {padding=<ltrb 0 0 4 0>}}
  }
  SizeBox_2 {
   widgetClassPath=&SizeBox
   parent=OptionSelectionArea
   widget=SizeBox_2
   properties {bOverride_WidthOverride=true, clipping=ClipToBounds, widthOverride=350}
   slot {class=HorizontalBoxSlot, properties {verticalAlignment=VAlign_Center, size {sizeRule=Fill, value=1}}}
  }
  Border_4 {
   widgetClassPath=&Border
   parent=SizeBox_2
   widget=Border_4
   properties {
    horizontalAlignment=HAlign_Center
    background {
     bIsDynamicallyLoaded=false
     drawAs=Image
     imageType=NoImage
     mirroring=NoMirror
     resourceName=None
     resourceObject=null
     tiling=NoTile
     imageSize=<32 32>
     margin=<ltrb 0 0 0 0>
     outlineSettings {
      bUseBrushTransparency=false
      roundingType=HalfHeightRadius
      width=0
      color {colorUseRule=UseColor_Specified, specifiedColor=<rgba 0 0 0 0>}
      cornerRadii=<0 0 0 0>
     }
     tintColor {colorUseRule=UseColor_Specified, specifiedColor=<rgba 0 0.0244 0.0625 0>}
     uVRegion {bIsValid=false, max=<0 0>, min=<0 0>}
    }
    padding=<ltrb 0 0 0 0>
   }
   slot {class=SizeBoxSlot}
  }
  MyText {
   widgetClassPath=&CommonTextBlock
   bInherited=true
   bIsVariable=true
   parent=Border_4
   widget=MyText
   properties {
    scrollStyle=&TextScrollStyle-Base_C
    text=" Option Text Á Option Text Á"
    colorAndOpacity {colorUseRule=UseColor_Specified, specifiedColor=<rgba 0.3176 0.7176 0.9843 1>}
    font {
     bForceMonospaced=false
     bMaterialIsStencil=false
     fontMaterial=null
     fontObject=&BurbankSmall
     letterSpacing=0
     monospacedWidth=1
     size=20
     skewAmount=0
     typefaceFontName=Bold
     outlineSettings {
      bApplyOutlineToDropShadows=false
      bMiteredCorners=false
      bSeparateFillAlpha=false
      outlineMaterial=null
      outlineSize=0
      outlineColor=<rgba 0 0 0 0>
     }
    }
    shadowColorAndOpacity=<rgba 0.0052 0.0052 0.0056 0>
   }
   slot {class=BorderSlot, properties {horizontalAlignment=HAlign_Center, padding=<ltrb 0 0 0 0>}}
  }
  RightButtonSize {
   widgetClassPath=&SizeBox
   bIsVariable=true
   parent=OptionSelectionArea
   widget=RightButtonSize
   properties {bOverride_HeightOverride=true, bOverride_WidthOverride=true, heightOverride=30, widthOverride=24}
   slot {class=HorizontalBoxSlot, properties {horizontalAlignment=HAlign_Center, verticalAlignment=VAlign_Center}}
  }
  RightButton {
   widgetClassPath=&SimpleCommonButton
   bIsVariable=true
   parent=RightButtonSize
   widget=RightButton
   properties {
    onButtonBaseClicked=(&self:WidgetTree.RightButton.HandleOnButtonClicked)
    onButtonBaseDoubleClicked=(&self:WidgetTree.RightButton.HandleOnButtonDoubleClicked)
    onButtonBaseHovered=(&self:WidgetTree.RightButton.HandleOnButtonHovered)
    onButtonBaseUnhovered=(&self:WidgetTree.RightButton.HandleOnButtonUnhovered)
    onSelectedChangedBase=(&self:WidgetTree.RightButton.HandleOnSelectedChanged)
    style=/FortUILibrary/Foundation/Buttons/ButtonStyles/ButtonStyle_SimpleCycleArrow_Right.ButtonStyle_SimpleCycleArrow_Right_C
    visibility=Hidden
   }
   slot {class=SizeBoxSlot, properties {padding=<ltrb 4 0 0 0>}}
  }
  SizeBox_TextEntry {
   widgetClassPath=&SizeBox
   bInherited=true
   bIsVariable=true
   parent=Overlay_1
   widget=SizeBox_TextEntry
   properties {maxDesiredWidth=750, visibility=Collapsed}
   slot {class=OverlaySlot, properties {horizontalAlignment=HAlign_Fill, verticalAlignment=VAlign_Fill}}
  }
  FilteredTextEntry_Widget {
   widgetClassPath=/FortUILibrary/UIKit/InputField/WBP_UIKit_FilteredInputField
   bInherited=true
   bIsVariable=true
   parent=SizeBox_TextEntry
   widget=FilteredTextEntry_Widget
   properties {bIsFocusable=true}
   slot {class=SizeBoxSlot}
  }
  SizeBox_Button {
   widgetClassPath=&SizeBox
   bIsVariable=true
   parent=Overlay_1
   widget=SizeBox_Button
   properties {maxDesiredWidth=750, visibility=Collapsed}
   slot {class=OverlaySlot, properties {horizontalAlignment=HAlign_Fill, verticalAlignment=VAlign_Fill}}
  }
  Button_Widget {
   widgetClassPath=&IconTextButton
   bIsVariable=true
   parent=SizeBox_Button
   widget=Button_Widget
   properties {
    onButtonBaseClicked=(&self:WidgetTree.Button_Widget.HandleOnButtonClicked)
    onButtonBaseDoubleClicked=(&self:WidgetTree.Button_Widget.HandleOnButtonDoubleClicked)
    onButtonBaseHovered=(&self:WidgetTree.Button_Widget.HandleOnButtonHovered)
    onButtonBaseUnhovered=(&self:WidgetTree.Button_Widget.HandleOnButtonUnhovered)
    onSelectedChangedBase=(&self:WidgetTree.Button_Widget.HandleOnSelectedChanged)
    style=/FortUILibrary/Foundation/Buttons/ButtonStyles/Athena/Athena_ButtonStyle_AngledBlueMenuButton.Athena_ButtonStyle_AngledBlueMenuButton_C
   }
   slot {class=SizeBoxSlot}
  }
  InputActionWidget {widgetClassPath=&CommonActionWidget, bInherited=true, widget=null}
 ]
}
