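# Modal widget, subclassing FortCustomMatchmakingKeyModal, where the party leader enters, clears or unmasks a custom matchmaking key; non-leaders see placeholder text with editing disabled.
# Activation sets field and button states from party leadership and encrypted-payload checks.
# Text edits enforce runtime length limits through UpdateLimitsText, which drives the colored counter and Accept state, shaking on trim.
# Accept submits changed keys asynchronously behind a throbber, closing on success or showing an error on failure.
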WidgetBlueprint /Game/Athena/UI/Matchmaking/CustomMatchmakingKeyModal
animations [Intro {
  binding_count=1
  length_seconds=0.25
  object=&self:Intro
  sequence {
   bindings [Overlay_Content {
     id=#1
     possessed_class=Overlay
     tracks [
      {
       type=MovieSceneFloatTrack
       display_name=Render Opacity
       sections [{type=MovieSceneFloatSection, end_frame=30, start_frame=0, channels [None {type=MovieSceneScriptingFloatChannel, keys [[0,0],[12,0],[30,1]]}]}]
      }
      {
       type=MovieScene2DTransformTrack
       sections [{
         type=MovieScene2DTransformSection
         end_frame=30
         start_frame=0
         channels [
          Scale.X {type=MovieSceneScriptingFloatChannel, keys [[12,1.15],[24,1]]}
          Scale.Y {type=MovieSceneScriptingFloatChannel, keys [[12,1.15],[24,1]]}
         ]
        }]
      }
     ]
    }]
   display_rate {numerator=120}
  }
 }]
graphs [
 BP_GetDesiredFocusTarget {nodes [
   Get Desired Focus Target()
   n1 = Get FortPartyContext!() -> (ReturnValue:Fort Party Context Object Reference) {is=title K2Node_GetBlueprintContextEx}
   n2 = Get EditableText_MatchmakingKey() -> (EditableText_MatchmakingKey:Editable Text Object Reference)
   n3 = GetIsEnabled!(self:Widget Object Reference=@n2) -> (ReturnValue:bool)
   n4 = Get ClearButton() -> (ClearButton:WBP UIKit Button Contextual Object Reference)
   > Return Node!(
    execute
    ReturnValue:Widget Object Reference=Select(
     Option 0:Widget Object Reference=Get CloseButton() -> (CloseButton:Fort CTAButton Object Reference)
     Option 1:Widget Object Reference=Select(Option 0=@n4, Option 1=@n2, Index=@n3)
     Index=IsLocalPlayerPartyLeader!(self:Fort Party Context Object Reference=@n1) -> (ReturnValue:bool)
    ) -> (ReturnValue:Widget Object Reference)
   )
  ]}
 EventGraph {nodes [
   "On Clicked (UnhideButton)"!() -> (OutputDelegate, then, Button:Common Button Base Object Reference) {is=title K2Node_ComponentBoundEvent}
   n1 = Get EditableText_MatchmakingKey()
   > Branch(Condition=GetIsEnabled(self=@n1))
   > SetIsPassword!(
    execute
    self:Editable Text Object Reference=@n1
    InbIsPassword:bool=NOT Boolean(A=Get IsPassword!(self:Editable Text Object Reference=@n1) -> (IsPassword:bool))
   ) -> (then)
   "OnTextChanged (Editable Text) (EditableText_MatchmakingKey)"!() -> (OutputDelegate, then, Text:"Text (by ref)") {is=title K2Node_ComponentBoundEvent}
   > n2 = LimitTextLength!(
    execute
    EditableTextWidget:Widget Object Reference=Get EditableText_MatchmakingKey()
    Length:int=Get MaximumCustomMatchmakingKeyLength!(self:Fort Runtime Options Object Reference=GetRuntimeOptions()) -> (MaximumCustomMatchmakingKeyLength:int=0)
   ) -> (then, WasTrimmed:bool)
   > n3 = UpdateLimitsText!(execute, self) -> (then, KeyIsValid:bool)
   > SetIsEnabled(self=Get AcceptButton() -> (AcceptButton:Fort CTAButton Object Reference), bInIsEnabled=@n3)
   n4 = Get bShakeAnimationIsPlaying() -> (bShakeAnimationIsPlaying:bool)
   > Branch(Condition=AND Boolean(A=@n2, B=NOT Boolean(A=@n4)))
   > Set bShakeAnimationIsPlaying(bShakeAnimationIsPlaying:bool=true) -> (Output_Get:bool)
   > PlayAnimationByName!(
    execute
    self:Fort Simple Widget Animations Panel Object Reference=Get SimpleWidgetAnimationsPanel() -> (SimpleWidgetAnimationsPanel:Fort Simple Widget Animations Panel Object Reference)
    Name:"Name (by ref)"=Get ShakeAnimation() -> (ShakeAnimation:Name=None)
    Speed:float=1
    StartAtTime:float=0
   ) -> (then, ReturnValue:bool)
   > Delay(Duration=0.5)
   > Set bShakeAnimationIsPlaying(bShakeAnimationIsPlaying=false)
   n5 = "On Clicked (AcceptButton)"!() -> (OutputDelegate, then, Button:Common Button Base Object Reference) {is=title K2Node_ComponentBoundEvent}
   > Fire Event Element Interacted(Element=@n5.Button)
   > SetIsEnabled(self=@n5.Button, bInIsEnabled=false)
   > SetIsEnabled(self=Get EditableText_MatchmakingKey(), bInIsEnabled=false)
   > SetIsEnabled(self=Get ClearButton(), bInIsEnabled=false)
   > UpdateDesiredUnhideButtonState!(execute, self, DesiredState:bool=false) -> (then, FinalState:bool)
   n6 = GetCustomMatchmakingKey!(self:Fort Matchmaking Context Object Reference=Get FortMatchmakingContext!() -> (ReturnValue:Fort Matchmaking Context Object Reference) {is=title K2Node_GetBlueprintContextEx}) -> (ReturnValue:String)
   n7 = "GetText (Editable Text)"!(self:Editable Text Object Reference=Get EditableText_MatchmakingKey()) -> (ReturnValue:Text)
   n8 = "To String (Text)"(InText=@n7)
   n9 = Get KeyIsEncryptedText() -> (KeyIsEncryptedText:Text)
   n10 = Get NotPartyLeaderText() -> (NotPartyLeaderText:Text)
   n11 = Get PartyLeaderHasSetKeyText() -> (PartyLeaderHasSetKeyText:Text)
   > n12 = Branch(Condition=AND Boolean!(
     A="Not Equal Exactly (String)"(A=@n6, B=@n8)
     B="Not Equal Exactly (Text)"!(A:"Text (by ref)"=@n7, B:"Text (by ref)"=@n9) -> (ReturnValue:bool)
     C:bool="Not Equal Exactly (Text)"(A=@n7, B=@n10)
     D:bool="Not Equal Exactly (Text)"(A=@n7, B=@n11)
    ) -> (ReturnValue))
   > SetVisibility(self=Get RequestPendingThrobber() -> (RequestPendingThrobber:User Widget Object Reference), InVisibility=SelfHitTestInvisible)
   n13 = Get EditableText_MatchmakingKey()
   > SetCustomMatchmakingKey!(
    execute
    self:Fort Matchmaking Context Object Reference=Get FortMatchmakingContext()
    Key:String="To String (Text)"(InText="GetText (Editable Text)"(self=@n13))
    Callback:"Delegate (by ref)"=Create Event()
   ) -> (then)
   DeactivateWidget(execute=@n12.else)
   n14 = "On Clicked (CloseButton)"!() -> (OutputDelegate, then, Button:Common Button Base Object Reference) {is=title K2Node_ComponentBoundEvent}
   > Fire Event Element Interacted(Element=@n14.Button)
   > DeactivateWidget()
   "On Clicked (ClearButton)"!() -> (OutputDelegate, then, Button:Common Button Base Object Reference) {is=title K2Node_ComponentBoundEvent}
   > "SetText (Editable Text)"!(execute, self:Editable Text Object Reference=Get EditableText_MatchmakingKey(), InText:Text=MakeLiteralText()) -> (then)
   n15 = Get EditableText_MatchmakingKey()
   > SetIsEnabled(self=@n15, bInIsEnabled=true)
   > SetIsEnabled(self=Get UnhideButton() -> (UnhideButton:WBP UIKit Button Contextual Object Reference), bInIsEnabled=true)
   > SetIsPassword(self=@n15, InbIsPassword=true)
   > n16 = UpdateLimitsText()
   > SetIsEnabled(self=Get AcceptButton(), bInIsEnabled=@n16)
   n17 = OnCustomMatchmakingKeySet!() -> (OutputDelegate, then, bSucceeded:bool=~, Error:String) {is=title K2Node_CustomEvent}
   > SetIsEnabled(self=Get AcceptButton(), bInIsEnabled=true)
   > SetIsEnabled(self=Get EditableText_MatchmakingKey(), bInIsEnabled=true)
   > SetIsEnabled(self=Get ClearButton(), bInIsEnabled=true)
   > SetVisibility(self=Get RequestPendingThrobber(), InVisibility=Hidden)
   > n18 = Branch(Condition=@n17.bSucceeded)
   > DeactivateWidget()
   Branch(execute=@n18.else, Condition=NOT Boolean(A=Get bShakeAnimationIsPlaying()))
   > Set bShakeAnimationIsPlaying(bShakeAnimationIsPlaying=true)
   > PlayAnimationByName(self=Get SimpleWidgetAnimationsPanel(), Name=Get ShakeAnimation() -> (ShakeAnimation=None), Speed=1, StartAtTime=0)
   > Delay(Duration=0.5)
   > Set bShakeAnimationIsPlaying(bShakeAnimationIsPlaying=false)
   > ShowErrorDialog!(
    execute
    DisplayMessage:Text="LOCTEXT(\"Failed to set custom matchmaking key.\")"
    ErrorCode:String=@n17.Error
    ErrorSeverity:EFortErrorSeverity Enum=Passive
   ) -> (then)
   Event On Activated()
   > SetVisibility(self=Get RequestPendingThrobber(), InVisibility=Hidden)
   > Fire Event Sub View Opened!(execute, View:Widget Object Reference=Self-Reference(), AdditionalInfo:Flow Analytics Additional Info Structure, bIsOverlay:bool=true) -> (then)
   > SetIsEnabled(self=Get ClearButton(), bInIsEnabled=true)
   n19 = Get FortMatchmakingContext()
   n20 = IsCustomMatchmakingKeyEncryptedPayload!(self:Fort Matchmaking Context Object Reference=@n19) -> (ReturnValue:bool)
   > n21 = Branch(Condition=@n20)
   n22 = Get EditableText_MatchmakingKey()
   > "SetText (Editable Text)"(self=@n22, InText=Get KeyIsEncryptedText())
   n23 = Get EditableText_MatchmakingKey()
   > SetIsEnabled(self=@n23, bInIsEnabled=false)
   > n24 = SetFocus(self=Get CloseButton())
   SetIsEnabled(execute=[@n33.then, @n24.then], self=Get AcceptButton(), bInIsEnabled=false)
   n25 = NOT Boolean(A=@n20)
   > UpdateDesiredUnhideButtonState(DesiredState=@n25)
   > SetIsPassword(self=Get EditableText_MatchmakingKey(), InbIsPassword=@n25)
   > n26 = Branch(Condition=IsLocalPlayerPartyLeader(self=Get FortPartyContext()))
   n27 = Self-Reference()
   > n28 = FortShippingLog!(
    execute
    WorldContextObject:Object Reference=@n27
    InString:String="owning local player IS the party leader - allowing custom match key options to remain! "
    bLogAsWarning:bool=false
   ) -> (then)
   UpdateLimitsText(execute=[@n32.then, @n28.then])
   n29 = GetRuntimeOptions()
   > SetVisibility(
    self=Get UnhideButton()
    InVisibility=Select(
     Option 0:ESlateVisibility Enum=Collapsed
     Option 1:ESlateVisibility Enum=SelfHitTestInvisible
     Index=Get bEnableEncryptedCustomMatchPayloads!(self:Fort Runtime Options Object Reference=@n29) -> (bEnableEncryptedCustomMatchPayloads:bool)
    ) -> (ReturnValue:ESlateVisibility Enum)
   )
   > PlayAnimationForward(InAnimation=Get Intro())
   FortShippingLog(
    execute=@n26.else
    WorldContextObject=@n27
    InString="owning local player is not the party leader - disabling custom match key options!"
    bLogAsWarning=false
   )
   > SetIsEnabled(self=Get AcceptButton(), bInIsEnabled=false)
   > SetIsEnabled(self=Get ClearButton(), bInIsEnabled=false)
   > SetIsEnabled(self=Get UnhideButton(), bInIsEnabled=false)
   n30 = Get EditableText_MatchmakingKey()
   > SetIsEnabled(self=@n30, bInIsEnabled=false)
   n31 = "To String (Text)"(InText="GetText (Editable Text)"(self=@n30))
   > "SetText (Editable Text)"(
    self=@n30
    InText=Select(Option 0:Text=Get PartyLeaderHasSetKeyText(), Option 1:Text=Get NotPartyLeaderText(), Index=IsEmpty(InString=@n31)) -> (ReturnValue:Text)
   )
   > n32 = SetIsPassword(self=@n30, InbIsPassword=false)
   "SetText (Editable Text)"(execute=@n21.else, self=@n22, InText="To Text (String)"(InString=GetCustomMatchmakingKey(self=@n19)))
   > SetIsEnabled(self=@n23, bInIsEnabled=true)
   > n33 = SetFocus(self=@n23)
  ]}
 UpdateDesiredUnhideButtonState {nodes [
   n1 = UpdateDesiredUnhideButtonState!() -> (then, DesiredState:bool=~) {is=title K2Node_FunctionEntry}
   n2 = AND Boolean!(A=Get bEnableEncryptedCustomMatchPayloads(self=GetRuntimeOptions()), B=IsLocalPlayerPartyLeader(self=Get FortPartyContext()), C=@n1) -> (ReturnValue)
   > SetIsEnabled(self=Get UnhideButton(), bInIsEnabled=@n2)
   > Return Node!(execute, FinalState:bool=@n2)
  ]}
 UpdateLimitsText {nodes [
   UpdateLimitsText() {is=title K2Node_FunctionEntry}
   n1 = GetRuntimeOptions()
   n2 = Get MaximumCustomMatchmakingKeyLength(self=@n1) -> (MaximumCustomMatchmakingKeyLength=0)
   n3 = "GetText (Editable Text)"(self=Get EditableText_MatchmakingKey())
   n4 = Len(S="To String (Text)"(InText=@n3))
   n5 = "To Text (Integer)"(Value="Clamp (Integer)"(Value=@n4, Min=0, Max=@n2) -> (ReturnValue=0))
   n6 = OR Boolean!(
    A="Equal Exactly (Text)"!(A:"Text (by ref)"=@n3, B:"Text (by ref)"=Get KeyIsEncryptedText()) -> (ReturnValue:bool)
    B="Equal Exactly (Text)"(A=@n3, B=Get NotPartyLeaderText())
    C:bool="Equal Exactly (Text)"(A=@n3, B=Get PartyLeaderHasSetKeyText())
   ) -> (ReturnValue)
   n7 = Get TextCharacterCount() -> (TextCharacterCount:Common Text Object Reference)
   > "SetText (Text)"(
    self=@n7
    InText=Format Text!(
     Format="INVTEXT(\"{Current} / {Max}\")"
     Current:Text=Select(Option 0=@n5, Option 1="INVTEXT(\"??\")", Index=@n6)
     Max:Text="To Text (Integer)"(Value=@n2)
    ) -> (Result)
   )
   n8 = Get MinimumCustomMatchmakingKeyLength!(self:Fort Runtime Options Object Reference=@n1) -> (MinimumCustomMatchmakingKeyLength:int=0)
   n9 = AND Boolean!(A="integer > integer"(A=@n4), B=OR Boolean!(A="integer < integer"(A=@n4, B=@n8), B="integer > integer"(A=@n4, B=@n2)) -> (ReturnValue)) -> (ReturnValue)
   n10 = Get ErrorTextColor() -> (ErrorTextColor:Slate Color Structure)
   n11 = Get ValidTextColor() -> (ValidTextColor:Slate Color Structure)
   > SetColorAndOpacity(
    self=@n7
    InColorAndOpacity=Select(
     Option 0:Slate Color Structure=Select(Option 0=@n11, Option 1=@n10, Index=@n9)
     Option 1:Slate Color Structure=Get DefaultTextColor() -> (DefaultTextColor:Slate Color Structure)
     Index=@n6
    ) -> (ReturnValue:Slate Color Structure)
   )
   > Return Node!(execute, KeyIsValid:bool=NOT Boolean(A=@n9))
   IsCustomMatchmakingKeyEncryptedPayload(self=Get FortMatchmakingContext())
  ]}
]
variables [
 bShakeAnimationIsPlaying {type=boolean}
 DefaultTextColor {type=SlateColor, default {colorUseRule=UseColor_Specified, specifiedColor=<rgba 0.5776 0.5776 0.5776 1>}}
 ErrorTextColor {type=SlateColor, default {colorUseRule=UseColor_Specified, specifiedColor=<rgba 1 0.131 0.1329 1>}}
 KeyIsEncryptedText {type=string, default=(key is encrypted, press Clear to change it)}
 NotPartyLeaderText {type=string, default=(you are not the party leader)}
 PartyLeaderHasSetKeyText {type=string, default=(key set by party leader)}
 ShakeAnimation {type=string, default=Shake}
 ValidTextColor {type=SlateColor, default {colorUseRule=UseColor_Specified, specifiedColor=<rgba 0.0395 1 0.2307 1>}}
]
widgets {
 info {
  inheritedWidgetCount=4
  namedSlotCount=1
  parentClass=/Script/FortniteUI.FortCustomMatchmakingKeyModal
  rootWidgetClass=/FortUILibrary/Systems/Dialog/WBP_UIKit_Dialog_Internal.WBP_UIKit_Dialog_Internal_C
  widgetCount=22
 }
 widgets [
  WBP_UIKit_Dialog_Base {
   widgetClassPath=/FortUILibrary/Systems/Dialog/WBP_UIKit_Dialog_Internal
   bIsVariable=true
   widget=WBP_UIKit_Dialog_Base
   properties {dialogSIze {
     height Override {enable=false, value=544}
     max Aspect Ratio {enable=false, value=0}
     max Desired Height {enable=true, value=800}
     max Desired Width {enable=false, value=0}
     min Aspect Ratio {enable=false, value=0}
     min Desired Height {enable=false, value=0}
     min Desired Width {enable=false, value=0}
     width Override {enable=true, value=864}
    }}
  }
  Overlay_Content {widgetClassPath=&Overlay, namedSlotHost=WBP_UIKit_Dialog_Base, widget=Overlay_Content}
  VerticalBox_1 {
   widgetClassPath=&VerticalBox
   parent=Overlay_Content
   widget=VerticalBox_1
   slot {class=OverlaySlot, properties {horizontalAlignment=HAlign_Fill, verticalAlignment=VAlign_Fill, padding=<ltrb 80 56 80 56>}}
  }
  Title {
   widgetClassPath=&CommonTextBlock
   bIsVariable=true
   parent=VerticalBox_1
   widget=Title
   properties {
    applyLineHeightToBottomLine=false
    autoWrapText=true
    text=CUSTOM MATCHMAKING KEY
    textTransformPolicy=ToUpper
    font {
     bForceMonospaced=false
     bMaterialIsStencil=false
     fontMaterial=null
     fontObject=&HeadingNow
     letterSpacing=-10
     monospacedWidth=1
     size=25.5
     skewAmount=0
     typefaceFontName=86 Bold
     outlineSettings {
      bApplyOutlineToDropShadows=false
      bMiteredCorners=false
      bSeparateFillAlpha=false
      outlineMaterial=null
      outlineSize=0
      outlineColor=<rgba 0 0 0 1>
     }
    }
    shadowOffset=<0 0>
    strikeBrush {
     bIsDynamicallyLoaded=false
     drawAs=NoDrawType
     imageType=NoImage
     mirroring=NoMirror
     resourceName=None
     resourceObject=null
     tiling=NoTile
     imageSize=<32 32>
     margin=<ltrb 0 0 0 0>
     outlineSettings {
      bUseBrushTransparency=false
      roundingType=HalfHeightRadius
      width=0
      color {colorUseRule=UseColor_Specified, specifiedColor=<rgba 0 0 0 0>}
      cornerRadii=<0 0 0 0>
     }
     tintColor {colorUseRule=UseColor_Specified, specifiedColor=<rgba 1 1 1 1>}
     uVRegion {bIsValid=false, max=<0 0>, min=<0 0>}
    }
   }
   slot {class=VerticalBoxSlot, properties {verticalAlignment=VAlign_Top, padding=<ltrb 0 0 0 32>}}
  }
  Content {
   widgetClassPath=&VerticalBox
   parent=VerticalBox_1
   widget=Content
   properties {navigation=&self:WidgetTree.Content.WidgetNavigation_1}
   slot {class=VerticalBoxSlot}
  }
  CommonTextBlock_47 {
   widgetClassPath=&CommonTextBlock
   parent=Content
   widget=CommonTextBlock_47
   properties {
    applyLineHeightToBottomLine=false
    autoWrapText=true
    lineHeightPercentage=1.05
    text="Setting a custom matchmaking key forces the matchmaking service to only match you with others also using the same key. This is used for private matches, tournaments, or other custom games.\r\n\r\nOnly the party leader can set a custom matchmaking key for the party."
    colorAndOpacity {colorUseRule=UseColor_Specified, specifiedColor=<rgba 0.5776 0.5776 0.5776 1>}
    font {
     bForceMonospaced=false
     bMaterialIsStencil=false
     fontMaterial=null
     fontObject=&HeadingNow
     letterSpacing=10
     monospacedWidth=1
     size=18
     skewAmount=0
     typefaceFontName=74 Regular
     outlineSettings {
      bApplyOutlineToDropShadows=false
      bMiteredCorners=false
      bSeparateFillAlpha=false
      outlineMaterial=null
      outlineSize=0
      outlineColor=<rgba 0 0 0 1>
     }
    }
    shadowOffset=<0 0>
    strikeBrush {
     bIsDynamicallyLoaded=false
     drawAs=NoDrawType
     imageType=NoImage
     mirroring=NoMirror
     resourceName=None
     resourceObject=null
     tiling=NoTile
     imageSize=<32 32>
     margin=<ltrb 0 0 0 0>
     outlineSettings {
      bUseBrushTransparency=false
      roundingType=HalfHeightRadius
      width=0
      color {colorUseRule=UseColor_Specified, specifiedColor=<rgba 0 0 0 0>}
      cornerRadii=<0 0 0 0>
     }
     tintColor {colorUseRule=UseColor_Specified, specifiedColor=<rgba 1 1 1 1>}
     uVRegion {bIsValid=false, max=<0 0>, min=<0 0>}
    }
   }
   slot {class=VerticalBoxSlot, properties {verticalAlignment=VAlign_Top, padding=<ltrb 0 0 0 16>}}
  }
  HorizontalBox_345 {widgetClassPath=&HorizontalBox, parent=Content, widget=HorizontalBox_345, slot {class=VerticalBoxSlot}}
  TextCharacterCount {
   widgetClassPath=&CommonTextBlock
   bIsVariable=true
   parent=HorizontalBox_345
   widget=TextCharacterCount
   properties {
    bIsScrollingEnabled=false
    text=0 / 16
    colorAndOpacity {colorUseRule=UseColor_Specified, specifiedColor=<rgba 1 0 1 1>}
    font {
     bForceMonospaced=false
     bMaterialIsStencil=false
     fontMaterial=null
     fontObject=&HeadingNow
     letterSpacing=0
     monospacedWidth=1
     size=24
     skewAmount=0
     typefaceFontName=76 Bold
     outlineSettings {
      bApplyOutlineToDropShadows=false
      bMiteredCorners=false
      bSeparateFillAlpha=false
      outlineMaterial=null
      outlineSize=0
      outlineColor=<rgba 0 0 0 1>
     }
    }
    shadowOffset=<0 0>
   }
   slot {class=HorizontalBoxSlot, properties {horizontalAlignment=HAlign_Right, padding=<ltrb 0 8 0 8>, size {sizeRule=Fill, value=1}}}
  }
  HorizontalBox_138 {widgetClassPath=&HorizontalBox, parent=Content, widget=HorizontalBox_138, slot {class=VerticalBoxSlot}}
  SimpleWidgetAnimationsPanel {
   widgetClassPath=/Script/FortniteUI.FortSimpleWidgetAnimationsPanel
   bIsVariable=true
   parent=HorizontalBox_138
   widget=SimpleWidgetAnimationsPanel
   properties {animations [Shake {
      bIsEnabled=true
      onAnimationsFinished=()
      targets [{
        alpha=null
        bResetOnFinish=true
        bShouldAnimateAlpha=false
        bShouldAnimateScale=false
        bShouldAnimateTranslation=true
        scale=null
        targetName=CommonBorder_2
        translation=/Game/UI/Utilities/Curve_ShakeAnim
        materialScalarAnimations []
       }]
     }]}
   slot {class=HorizontalBoxSlot, properties {padding=<ltrb 0 0 8 0>, size {sizeRule=Fill, value=1}}}
  }
  CommonBorder_2 {
   widgetClassPath=&CommonBorder
   parent=SimpleWidgetAnimationsPanel
   widget=CommonBorder_2
   properties {
    verticalAlignment=VAlign_Top
    background {
     bIsDynamicallyLoaded=false
     drawAs=Image
     imageType=NoImage
     mirroring=NoMirror
     resourceName=None
     resourceObject=null
     tiling=NoTile
     imageSize=<32 32>
     margin=<ltrb 0 0 0 0>
     outlineSettings {
      bUseBrushTransparency=false
      roundingType=HalfHeightRadius
      width=0
      color {colorUseRule=UseColor_Specified, specifiedColor=<rgba 0 0 0 0>}
      cornerRadii=<0 0 0 0>
     }
     tintColor {colorUseRule=UseColor_Specified, specifiedColor=<rgba 1 1 1 1>}
     uVRegion {bIsValid=false, max=<0 0>, min=<0 0>}
    }
    padding=<ltrb 0 0 0 0>
   }
   slot {class=FortSimpleWidgetAnimationsSlot}
  }
  HorizontalBox_2 {
   widgetClassPath=&HorizontalBox
   parent=CommonBorder_2
   widget=HorizontalBox_2
   slot {class=BorderSlot, properties {verticalAlignment=VAlign_Top, padding=<ltrb 0 0 0 0>}}
  }
  CommonVisibilityWidget_0 {
   widgetClassPath=&CommonVisibilityWidgetLegacy
   parent=HorizontalBox_2
   widget=CommonVisibilityWidget_0
   properties {
    bShowForAndroid=false
    bShowForIOS=false
    bShowForMac=false
    bShowForMouseAndKeyboard=false
    bShowForPC=false
    bShowForTouch=false
    padding=<ltrb 0 0 0 0>
   }
   slot {class=HorizontalBoxSlot, properties {padding=<ltrb 2 2 2 2>}}
  }
  EditTextActionWidget {
   widgetClassPath=&CommonActionWidget
   bIsVariable=true
   parent=CommonVisibilityWidget_0
   widget=EditTextActionWidget
   properties {inputActions [{dataTable=&InputActionDataTable, rowName=GenericForward}]}
   slot {class=BorderSlot, properties {padding=<ltrb 0 0 0 0>}}
  }
  EditableText_MatchmakingKey {
   widgetClassPath=/Script/UMG.EditableText
   bInherited=true
   bIsVariable=true
   parent=HorizontalBox_2
   widget=EditableText_MatchmakingKey
   properties {
    isPassword=true
    text=Test
    widgetStyle {
     backgroundImageComposing {
      bIsDynamicallyLoaded=false
      drawAs=Border
      imageType=FullColor
      mirroring=NoMirror
      resourceName=../../../Engine/Content/Slate/Old/HyperlinkDotted.png
      resourceObject=null
      tiling=Both
      imageSize=<0 0>
      margin=<ltrb 0 0 0 0.1875>
      outlineSettings {
       bUseBrushTransparency=false
       roundingType=HalfHeightRadius
       width=0
       color {colorUseRule=UseColor_Specified, specifiedColor=<rgba 0 0 0 0>}
       cornerRadii=<0 0 0 0>
      }
      tintColor {colorUseRule=UseColor_Specified, specifiedColor=<rgba 1 1 1 1>}
      uVRegion {bIsValid=false, max=<0 0>, min=<0 0>}
     }
     backgroundImageSelected {
      bIsDynamicallyLoaded=false
      drawAs=Box
      imageType=FullColor
      mirroring=NoMirror
      resourceName=../../../Engine/Content/Slate/Common/EditableTextSelectionBackground.png
      resourceObject=null
      tiling=NoTile
      imageSize=<0 0>
      margin=<ltrb 0.25 0.25 0.25 0.25>
      outlineSettings {
       bUseBrushTransparency=false
       roundingType=HalfHeightRadius
       width=0
       color {colorUseRule=UseColor_Specified, specifiedColor=<rgba 0 0 0 0>}
       cornerRadii=<0 0 0 0>
      }
      tintColor {colorUseRule=UseColor_Specified, specifiedColor=<rgba 1 1 1 1>}
      uVRegion {bIsValid=false, max=<0 0>, min=<0 0>}
     }
     caretImage {
      bIsDynamicallyLoaded=false
      drawAs=Image
      imageType=FullColor
      mirroring=NoMirror
      resourceName=../../../Engine/Content/Slate/Old/White.png
      resourceObject=null
      tiling=NoTile
      imageSize=<16 16>
      margin=<ltrb 0 0 0 0>
      outlineSettings {
       bUseBrushTransparency=false
       roundingType=HalfHeightRadius
       width=0
       color {colorUseRule=UseColor_Specified, specifiedColor=<rgba 0 0 0 0>}
       cornerRadii=<0 0 0 0>
      }
      tintColor {colorUseRule=UseColor_Specified, specifiedColor=<rgba 1 1 1 1>}
      uVRegion {bIsValid=false, max=<0 0>, min=<0 0>}
     }
     colorAndOpacity {colorUseRule=UseColor_Specified, specifiedColor=<rgba 0 0 0 1>}
     font {
      bForceMonospaced=false
      bMaterialIsStencil=false
      fontMaterial=null
      fontObject=&HeadingNow
      letterSpacing=0
      monospacedWidth=2.946
      size=15
      skewAmount=0
      typefaceFontName=67 Extra Bold
      outlineSettings {
       bApplyOutlineToDropShadows=false
       bMiteredCorners=false
       bSeparateFillAlpha=false
       outlineMaterial=null
       outlineSize=0
       outlineColor=<rgba 0 0 0 1>
      }
     }
     hintColorAndOpacity {colorUseRule=UseColor_Foreground, specifiedColor=<rgba 1 0 1 1>}
     selectedBackgroundColor {colorUseRule=UseColor_Foreground, specifiedColor=<rgba 1 0 1 1>}
    }
   }
   slot {class=HorizontalBoxSlot, properties {verticalAlignment=VAlign_Center, padding=<ltrb 16 0 16 0>, size {sizeRule=Fill, value=1}}}
  }
  Spacer_EditText {
   widgetClassPath=&Spacer
   parent=HorizontalBox_2
   widget=Spacer_EditText
   properties {size=<0 56>}
   slot {class=HorizontalBoxSlot, properties {horizontalAlignment=HAlign_Left, verticalAlignment=VAlign_Top}}
  }
  UnhideButton {
   widgetClassPath=/FortUILibrary/UIKit/Buttons/WBP_UIKit_Button_Contextual
   bIsVariable=true
   parent=HorizontalBox_138
   widget=UnhideButton
   properties {
    backgroundClass=/FortUILibrary/Systems/ModularButton/Blocks/WBP_UIKit_Block_BackgroundGeneric.WBP_UIKit_Block_BackgroundGeneric_C
    backgroundMaterial=null
    icon=/Game/Athena/UI/InGame/Creative/PropertyEditor/Icon_Picker/Textures_V2/64x64/T_UI_IconLibrary_Eye_64
    outlineMaterial=null
    style=/Game/UI/Foundation/Modals/PasscodeModal/Styles/ButtonStyle_Empty.ButtonStyle_Empty_C
    useDisabledVisualsOnLocked=true
    buttonPadding=<ltrb 12 0 12 0>
    foregroundColor {colorUseRule=UseColor_Foreground, specifiedColor=<rgba 1 0.8548 0.8457 1>}
    iconSize=<32 32>
   }
   slot {class=HorizontalBoxSlot, properties {padding=<ltrb 8 0 8 0>}}
  }
  ClearButton {
   widgetClassPath=&WBP_UIKit_Button_Contextual
   bIsVariable=true
   parent=HorizontalBox_138
   widget=ClearButton
   properties {
    icon=/Game/UI/Foundation/Textures/Buttons/T-SlotIcon-Remove-32
    useDisabledVisualsOnLocked=true
    buttonPadding=<ltrb 12 0 12 0>
    iconSize=<32 32>
   }
   slot {class=HorizontalBoxSlot, properties {padding=<ltrb 8 0 0 0>}}
  }
  HorizontalBox_0 {
   widgetClassPath=&HorizontalBox
   parent=VerticalBox_1
   widget=HorizontalBox_0
   slot {class=VerticalBoxSlot, properties {verticalAlignment=VAlign_Bottom, padding=<ltrb 0 40 0 0>}}
  }
  RequestPendingThrobber {
   widgetClassPath=/FortUILibrary/UIKit/Throbber/WBP_UIKit_Throbber
   bInherited=true
   bIsVariable=true
   parent=HorizontalBox_0
   widget=RequestPendingThrobber
   slot {class=HorizontalBoxSlot, properties {padding=<ltrb 0 0 24 0>}}
  }
  AcceptButton {
   widgetClassPath=&WBP_UIKit_Button_Regular
   bInherited=true
   bIsVariable=true
   parent=HorizontalBox_0
   widget=AcceptButton
   properties {minHeight=56, text=Accept, triggeringInputAction {dataTable=&InputActionDataTable, rowName=Accept}}
   slot {class=HorizontalBoxSlot, properties {verticalAlignment=VAlign_Center, padding=<ltrb 0 0 24 0>, size {sizeRule=Fill, value=1}}}
  }
  CloseButton {
   widgetClassPath=&WBP_UIKit_Button_Regular
   bInherited=true
   bIsVariable=true
   parent=HorizontalBox_0
   widget=CloseButton
   properties {buttonType=Quiet, minHeight=56, text=Close, triggeringInputAction {dataTable=&InputActionDataTable, rowName=Close}}
   slot {class=HorizontalBoxSlot, properties {verticalAlignment=VAlign_Center, size {sizeRule=Fill, value=1}}}
  }
 ]
}
