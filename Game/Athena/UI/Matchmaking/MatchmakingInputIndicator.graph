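# Matchmaking UI widget that shows an icon for an input type, built on MatchmakingInputIndicatorBase.
# On OnInputSourceTypeChanged it asks the FortGlobalUIContext subsystem for that input type's icon texture.
# A valid texture is set on the InputIcon image, which becomes hit-test invisible; otherwise the image collapses.
# The image starts collapsed with a controller icon brush, inside an inherited SizeBox root named SizeBox_HotfixVisibility.
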WidgetBlueprint /Game/Athena/UI/Matchmaking/MatchmakingInputIndicator
graphs [EventGraph {nodes [
   n1 = Event OnInputSourceTypeChanged!() -> (OutputDelegate, then, InputType:ECommonInputType Enum) {is=title K2Node_Event}
   > n2 = GetInputTypeIcon!(
    execute
    self:Fort Global UIContext Object Reference=Get FortGlobalUIContext() -> (ReturnValue:Fort Global UIContext Object Reference) {is=title K2Node_GetSubsystem}
    InputType:ECommonInputType Enum=@n1.InputType
   ) -> (then, ReturnValue:Texture 2D Object Reference)
   > n3 = Is Valid(InputObject=@n2)
   SetVisibility(execute=@n3.Is Not Valid, self=Get InputIcon() -> (InputIcon:Image Object Reference), InVisibility=Collapsed)
   SetBrushFromTexture!(execute=@n3.Is Valid, self:Image Object Reference=Get InputIcon(), Texture:Texture 2D Object Reference=@n2, bMatchSize:bool=false) -> (then)
   > SetVisibility(self=Get InputIcon(), InVisibility=HitTestInvisible)
  ]}]
variables []
widgets {
 info {inheritedWidgetCount=1, parentClass=/Script/FortniteUI.MatchmakingInputIndicatorBase, rootWidgetClass=&SizeBox, widgetCount=2}
 widgets [
  SizeBox_HotfixVisibility {widgetClassPath=&SizeBox, bInherited=true, widget=SizeBox_HotfixVisibility}
  InputIcon {
   widgetClassPath=&Image
   bIsVariable=true
   parent=SizeBox_HotfixVisibility
   widget=InputIcon
   properties {
    visibility=Collapsed
    brush {
     bIsDynamicallyLoaded=false
     drawAs=Image
     imageType=NoImage
     mirroring=NoMirror
     resourceName=None
     resourceObject=/Game/UI/Icons/Input/T_UI_Icon_Input_Controller_128
     tiling=NoTile
     imageSize=<64 64>
     margin=<ltrb 0 0 0 0>
     outlineSettings {
      bUseBrushTransparency=false
      roundingType=HalfHeightRadius
      width=0
      color {colorUseRule=UseColor_Specified, specifiedColor=<rgba 0 0 0 0>}
      cornerRadii=<0 0 0 0>
     }
     tintColor {colorUseRule=UseColor_Specified, specifiedColor=<rgba 1 1 1 1>}
     uVRegion {bIsValid=false, max=<0 0>, min=<0 0>}
    }
   }
   slot {class=SizeBoxSlot}
  }
 ]
}
